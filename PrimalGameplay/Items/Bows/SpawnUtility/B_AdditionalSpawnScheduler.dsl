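; Bow utility that schedules additional arrow spawns from an arrow spawn properties struct, signalling bound listeners (the original arrow) through the NotifySpawn dispatcher.
; StartSchedule calls StartSpawning immediately or after a timer delay.
; StartSpawning either broadcasts once for an immediate spawn, or runs DelayedSpawn now and on a looping timer, notifying one spawn per call.
; On completion it unbinds all listeners, with delayed runs also clearing the timer by function name.
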(event Custom|StartSchedule
  (bind _spawnproperties (Variables|Default|GetSpawnProperties))
  (bind (_numtospawn_21_f81b614a4aed43e706d812ac58824f1c _initialdelay_22_84d4e959441176c5e4d87db9e569192c _delaybetweenspawns_23_25624f804be8117272aacfa22e533989) (Utilities|Struct|BreakFArrowAdditionalSpawnProperties _spawnproperties))
  (bind _value (Utilities|Struct|BreakScalableFloat _delaybetweenspawns_23_25624f804be8117272aacfa22e533989))
  ;[1 1 1] If spawn should have initial delay, set timer; otherwise, start spawning immediately
  (if (> _value 0.0)
    (Utilities|Time|SetTimerbyEvent (delegate Custom|StartSpawning) _value)
    (else
      (CallFunction|StartSpawning))))

(event Custom|StartSpawning
  (bind _spawnproperties (Variables|Default|GetSpawnProperties))
  (bind (_numtospawn_21_f81b614a4aed43e706d812ac58824f1c _initialdelay_22_84d4e959441176c5e4d87db9e569192c _bdelaybetweenspawns_12_a60b1f414fc742e5113b81af854de145 _delaybetweenspawns_23_25624f804be8117272aacfa22e533989) (Utilities|Struct|BreakFArrowAdditionalSpawnProperties _spawnproperties))
  ;[1 1 1] If there should be a delay between each spawn, trigger spawn once then set timer for additional spawns; otherwise, trigger spawn immediately and unbind event
  (if _bdelaybetweenspawns_12_a60b1f414fc742e5113b81af854de145
    (CallFunction|DelayedSpawn)
    ;;[1 1 1] Set timer for additional spawns and cache handle so it can be cleared later
    (bind _returnvalue (Utilities|Time|SetTimerbyEvent (delegate Custom|DelayedSpawn) (Utilities|Struct|BreakScalableFloat _delaybetweenspawns_23_25624f804be8117272aacfa22e533989) true))
    (Variables|Default|SetCachedTimerHandle _returnvalue)
    (else
      ;;[1 1 1] Send event message to bound objects (the original arrow) then unbind
      (Default|CallNotifySpawn (Variables|Default|GetSpawnProperties))
      (Default|UnbindallEventsfromNotifySpawn)
      (Development|PrintString 0 "Schedule completed. Unbinding"))))

(event Custom|DelayedSpawn
  (bind _spawnproperties (Variables|Default|GetSpawnProperties))
  ;[1 1 1] Delayed Spawn Execution
  (Default|CallNotifySpawn _spawnproperties true 1)
  ;;[1 1 1] Clear timer and unbind all event listeners if we have spawned the correct amount
  (bind _result (Math|Integer|IncrementInt (Variables|Default|GetCurrentNumSpawns))
    (:%20%20
      (if (>= _result (Math|Float|Round (Utilities|Struct|BreakScalableFloat (Utilities|Struct|BreakFArrowAdditionalSpawnProperties _spawnproperties))))
        (Utilities|Time|ClearTimerbyFunctionName 0 "DelayedSpawn")
        (Default|UnbindallEventsfromNotifySpawn)
        (Development|PrintString 0 "Schedule completed. Unbinding")))))
